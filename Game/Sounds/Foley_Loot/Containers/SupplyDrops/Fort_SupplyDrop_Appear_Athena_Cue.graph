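# Sound cue for the appearance of a supply drop in Athena, played through the ambient sound class with supply-drop alert high-frequency attenuation.
# It picks one of two appear waves at equal weight, starts it after a short random delay and varies its pitch.
# Attenuation reaches a very long distance, virtualization is disabled, and a concurrency override caps simultaneous instances at two, stopping the farthest first.
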SoundCue /Game/Sounds/Foley_Loot/Containers/SupplyDrops/Fort_SupplyDrop_Appear_Athena_Cue
root=Modulator_0
nodes [
 {id=Delay_0, children [Random_0], properties {delayMax=2, delayMin=1}}
 {id=Modulator_0, children [Delay_0], properties {pitchMax=1.3, pitchMin=0.9, volumeMax=1, volumeMin=1}}
 {id=Random_0, children [WavePlayer_3,WavePlayer_0], properties {bShouldExcludeFromBranchCulling=true, weights [1,1]}}
 {id=WavePlayer_0, properties {soundWaveAssetPtr=&/SupplyDrop_Appear_02}}
 {id=WavePlayer_3, properties {soundWaveAssetPtr=&/SupplyDrop_Appear_01}}
]
properties {
 attenuationSettings=&/Fort_SupplyDrop_Alert_HighFreq_Athena_Att
 bOverrideConcurrency=true
 maxDistance=230000
 soundClassObject=&Fort_Ambient_Athena
 virtualizationMode=Disabled
 volumeMultiplier=1.504
 concurrencyOverrides {maxCount=2, resolutionRule=StopFarthestThenPreventNew}
}
