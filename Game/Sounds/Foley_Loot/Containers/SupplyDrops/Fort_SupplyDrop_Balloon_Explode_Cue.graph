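# Sound cue for a supply drop balloon exploding, layering a close explosion with a distant pop.
# A mixer plays both waves together, each through its own Athena attenuation node, with the distant layer slightly quieter.
# The cue routes to the large explosives sound class at reduced volume.
# A concurrency override allows one instance at a time and blocks new plays while it is active.
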SoundCue /Game/Sounds/Foley_Loot/Containers/SupplyDrops/Fort_SupplyDrop_Balloon_Explode_Cue
root=Mixer_1
nodes [
 {
  id=Attenuation_2
  children [WavePlayer_3]
  properties {attenuationSettings=&/Fort_SupplyDrop_Balloon_Athena_Att}
 }
 {
  id=Attenuation_3
  children [WavePlayer_2]
  properties {attenuationSettings=&/Fort_SupplyDrop_Balloon_Distant_Athena_Att}
 }
 {id=Mixer_1, children [Attenuation_2, Attenuation_3], properties {inputVolume [1,0.75]}}
 {id=WavePlayer_2, properties {soundWaveAssetPtr=&/sfx_supplydrop_balloonpop_distant}}
 {id=WavePlayer_3, properties {soundWaveAssetPtr=&/SupplyDrop_Balloon_Explode_01}}
]
properties {
 bOverrideConcurrency=true
 maxDistance=30000
 soundClassObject=/Game/Sounds/FortSoundClassesAndModes/Fort_Explosives_Large_PSM
 volumeMultiplier=0.2
 attenuationOverrides {bAttenuate=false}
 concurrencyOverrides {maxCount=1, resolutionRule=PreventNew}
}
