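# Sound cue for a falling Athena supply drop, spatialized with crate descent attenuation in the ambient sound class.
# A looping falling wave under a rising volume envelope mixes with randomized one-shot layers.
# Looped, delayed random picks supply strap and wood stress creaks and lower-pitched contents rattles, each with pitch and volume modulation.
# A quality-level node carries these detail layers in its first two slots only, leaving just the falling loop otherwise.
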SoundCue /Game/Sounds/Foley_Loot/Containers/SupplyDrops/Fort_SupplyDrop_Falling_Loop_Athena_Cue
root=Mixer_0
nodes [
 {id=Delay_2, children [Modulator_0]}
 {id=Delay_3, children [Modulator_1], properties {delayMax=2}}
 {
  id=Enveloper_0
  children [Modulator_2]
  properties {pitchCurve {editorCurveData {keys [{time=0}]}}, volumeCurve {editorCurveData {keys [<at 0.0162 0.0061>,<at 0.9994 1>]}}}
 }
 {id=Looping_0, children [Delay_2]}
 {id=Looping_1, children [Delay_3]}
 {id=Mixer_0, children [Enveloper_0, QualityLevel_1], properties {inputVolume [1.1,1.1]}}
 {id=Mixer_2, children [Looping_0,Looping_1], properties {inputVolume [1,1.4]}}
 {id=Modulator_0, children [Random_0], properties {pitchMax=1.2, pitchMin=0.8, volumeMax=1, volumeMin=0.7}}
 {id=Modulator_1, children [Random_1], properties {pitchMax=0.9, pitchMin=0.6, volumeMax=1, volumeMin=0.6}}
 {id=Modulator_2, children [WavePlayer_19], properties {pitchMax=0.95, pitchMin=0.85, volumeMax=1, volumeMin=1}}
 {id=QualityLevel_1, children [Mixer_2,Mixer_2,null,null]}
 {
  id=Random_0
  children [
   WavePlayer_3
   WavePlayer_4
   WavePlayer_5
   WavePlayer_6
   WavePlayer_7
   WavePlayer_8
   WavePlayer_9
   WavePlayer_10
   WavePlayer_11
   WavePlayer_12
  ]
  properties {bShouldExcludeFromBranchCulling=true, weights [1,1,1,1,1,1,1,1,1,1]}
 }
 {
  id=Random_1
  children [WavePlayer_13, WavePlayer_14, WavePlayer_15, WavePlayer_16, WavePlayer_17, WavePlayer_18]
  properties {bShouldExcludeFromBranchCulling=true, weights [1,1,1,1,1,1]}
 }
 {id=WavePlayer_10, properties {soundWaveAssetPtr=&/SupplyDrop_Wood_Stress_04}}
 {id=WavePlayer_11, properties {soundWaveAssetPtr=&/SupplyDrop_Wood_Stress_05}}
 {id=WavePlayer_12, properties {soundWaveAssetPtr=&/SupplyDrop_Wood_Stress_06}}
 {id=WavePlayer_13, properties {soundWaveAssetPtr=&/SupplyDrop_Contents_Rattle_01}}
 {id=WavePlayer_14, properties {soundWaveAssetPtr=&/SupplyDrop_Contents_Rattle_02}}
 {id=WavePlayer_15, properties {soundWaveAssetPtr=&/SupplyDrop_Contents_Rattle_03}}
 {id=WavePlayer_16, properties {soundWaveAssetPtr=&/SupplyDrop_Contents_Rattle_04}}
 {id=WavePlayer_17, properties {soundWaveAssetPtr=&/SupplyDrop_Contents_Rattle_05}}
 {id=WavePlayer_18, properties {soundWaveAssetPtr=&/SupplyDrop_Contents_Rattle_06}}
 {id=WavePlayer_19, properties {bLooping=true, soundWaveAssetPtr=&/SupplyDrop_Falling_Loop_01}}
 {id=WavePlayer_3, properties {soundWaveAssetPtr=&/SupplyDrop_Strap_Stress_01}}
 {id=WavePlayer_4, properties {soundWaveAssetPtr=&/SupplyDrop_Strap_Stress_02}}
 {id=WavePlayer_5, properties {soundWaveAssetPtr=&/SupplyDrop_Strap_Stress_03}}
 {id=WavePlayer_6, properties {soundWaveAssetPtr=&/SupplyDrop_Strap_Stress_04}}
 {id=WavePlayer_7, properties {soundWaveAssetPtr=&/SupplyDrop_Wood_Stress_01}}
 {id=WavePlayer_8, properties {soundWaveAssetPtr=&/SupplyDrop_Wood_Stress_02}}
 {id=WavePlayer_9, properties {soundWaveAssetPtr=&/SupplyDrop_Wood_Stress_03}}
]
properties {
 attenuationSettings=&/Fort_SupplyDrop_Crate_Descent_Athena_Att
 maxDistance=23000
 soundClassObject=&Fort_Ambient_Athena
 volumeMultiplier=1.002
}
