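# Sound cue for the looping flare sound on Athena supply drop markers.
# A single wave player loops SupplyDrop_Flare_Loop_01 directly as the root, with no mixing or randomization.
# It uses a close-range flare marker attenuation asset and the Athena ambient sound class at reduced volume.
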SoundCue /Game/Sounds/Foley_Loot/Containers/SupplyDrops/Fort_SupplyDrop_Flare_Loop_Athena_Cue
root=WavePlayer_1
nodes [{id=WavePlayer_1, properties {bLooping=true, soundWaveAssetPtr=&/SupplyDrop_Flare_Loop_01}}]
properties {
 attenuationSettings=&/Fort_SupplyDrop_Marker_Flare_Close_Athena_Att
 maxDistance=1500
 soundClassObject=&Fort_Ambient_Athena
 volumeMultiplier=0.351
}
