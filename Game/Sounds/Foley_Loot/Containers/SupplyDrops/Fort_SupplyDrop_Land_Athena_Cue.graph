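# Sound cue for the supply drop crate landing sound in Athena.
# It plays a single SupplyDrop_Land_01 wave through a modulator that randomizes pitch slightly while holding volume constant.
# It uses the crate descent attenuation and the Athena ambient sound class, with boosted volume.
# Concurrency override limits it to one instance per owner, refusing new plays, and virtualization is disabled.
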SoundCue /Game/Sounds/Foley_Loot/Containers/SupplyDrops/Fort_SupplyDrop_Land_Athena_Cue
root=Modulator_0
nodes [
 {id=Modulator_0, children [WavePlayer_4], properties {pitchMax=1.15, pitchMin=0.9, volumeMax=1, volumeMin=1}}
 {id=WavePlayer_4, properties {soundWaveAssetPtr=&/SupplyDrop_Land_01}}
]
properties {
 attenuationSettings=&/Fort_SupplyDrop_Crate_Descent_Athena_Att
 bOverrideConcurrency=true
 maxDistance=23000
 soundClassObject=&Fort_Ambient_Athena
 virtualizationMode=Disabled
 volumeMultiplier=1.504
 concurrencyOverrides {bLimitToOwner=true, maxCount=1, resolutionRule=PreventNew}
}
