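# Sound cue named for the last-second moment of a supply drop in Athena, in the container loot foley set.
# It has no root and no sound nodes, so as shown it produces no audio.
# It still carries settings: the crate-close attenuation asset, the Athena ambient sound class, a max distance and a boosted volume multiplier.
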SoundCue /Game/Sounds/Foley_Loot/Containers/SupplyDrops/Fort_SupplyDrop_LastSecond_Athena_Cue
root=null
nodes []
properties {
 attenuationSettings=&/Fort_SupplyDrop_Crate_Close_Athena_Att
 maxDistance=3000
 soundClassObject=&Fort_Ambient_Athena
 volumeMultiplier=6.237
}
