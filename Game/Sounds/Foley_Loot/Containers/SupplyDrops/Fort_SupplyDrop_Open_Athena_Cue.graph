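# Sound cue for opening a supply drop in Athena.
# A mixer layers the supply drop open wave with a wood close-impact wave borrowed from the demolition impact sounds, which starts after a short fixed delay.
# The cue has boosted volume, uses the loot reward sound class and takes its attenuation from the supply drop crate close Athena asset.
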SoundCue /Game/Sounds/Foley_Loot/Containers/SupplyDrops/Fort_SupplyDrop_Open_Athena_Cue
root=Mixer_1
nodes [
 {id=Delay_0, children [WavePlayer_10], properties {delayMax=0.2, delayMin=0.2}}
 {id=Mixer_1, children [WavePlayer_0,Delay_0], properties {inputVolume [1,1]}}
 {id=WavePlayer_0, properties {soundWaveAssetPtr=&/sfx_supplydrop_open}}
 {id=WavePlayer_10, properties {soundWaveAssetPtr=/Game/Sounds/Fort_Impact_Sounds/Demolition/Defaults/wood_close_impact_08}}
]
properties {
 attenuationSettings=&/Fort_SupplyDrop_Crate_Close_Athena_Att
 maxDistance=3000
 soundClassObject=&Fort_Loot_Reward_PSM
 volumeMultiplier=2
}
