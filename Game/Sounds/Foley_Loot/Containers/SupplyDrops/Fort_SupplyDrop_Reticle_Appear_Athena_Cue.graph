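# Sound cue for a supply drop reticle appearing in Athena.
# A single wave player plays the reticle-appear sound wave with boosted volume, a reticle alert attenuation asset and the critical in-game UI sound class.
# Virtualization is disabled, and a concurrency override caps simultaneous instances, stopping the farthest before preventing new ones.
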SoundCue /Game/Sounds/Foley_Loot/Containers/SupplyDrops/Fort_SupplyDrop_Reticle_Appear_Athena_Cue
root=WavePlayer_1
nodes [{id=WavePlayer_1, properties {soundWaveAssetPtr=&/sfx_supplydrop_reticle_appear}}]
properties {
 attenuationSettings=&/Fort_SupplyDrop_Reticle_Alert_Athena_Att
 bOverrideConcurrency=true
 maxDistance=75000
 soundClassObject=&Fort_UI_InGame_Critical
 virtualizationMode=Disabled
 volumeMultiplier=2
 concurrencyOverrides {maxCount=3, resolutionRule=StopFarthestThenPreventNew}
}
