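# Looping SoundCue for the supply drop target reticle in Athena, played as an ambient-class sound with close-range marker attenuation.
# A quality-level root sends its first two levels to one mixer and its last two to another.
# Each mixer layers a looping target sound with a quieter looping arrow sound.
# The second mixer swaps in an "opt" target wave and raises the arrow layer's volume.
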SoundCue /Game/Sounds/Foley_Loot/Containers/SupplyDrops/Fort_SupplyDrop_Reticle_Loop_Athena_Cue
root=QualityLevel_0
nodes [
 {id=Mixer_0, children [WavePlayer_0,WavePlayer_2], properties {inputVolume [1,0.35]}}
 {id=Mixer_1, children [WavePlayer_1,WavePlayer_3], properties {inputVolume [1,0.5]}}
 {id=QualityLevel_0, children [Mixer_0,Mixer_0,Mixer_1,Mixer_1]}
 {id=WavePlayer_0, properties {bLooping=true, soundWaveAssetPtr=&/sfx_supplydrop_target_loop}}
 {id=WavePlayer_1, properties {bLooping=true, soundWaveAssetPtr=&/sfx_supplydrop_target_loop_opt}}
 {
  id=WavePlayer_2
  properties {bLooping=true, soundWaveAssetPtr=&/SupplyDrop_Reticle_Arrow_Loop_1}
 }
 {
  id=WavePlayer_3
  properties {bLooping=true, soundWaveAssetPtr=&SupplyDrop_Reticle_Arrow_Loop_1}
 }
]
properties {
 attenuationSettings=&/Fort_SupplyDrop_Marker_Close_Athena_Att
 maxDistance=1500
 soundClassObject=&Fort_Ambient_Athena
 volumeMultiplier=2.079
}
