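# Sound cue for the supply drop reticle appearing in the Athena mode, on the ambient sound class.
# A single wave player feeds the reticle appear sound straight to the output.
# It uses a reticle alert attenuation asset with a very large max distance and a slight volume boost.
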SoundCue /Game/Sounds/Foley_Loot/Containers/SupplyDrops/Fort_SupplyDrop_Wave_Appear_Athena_Cue
root=WavePlayer_1
nodes [{id=WavePlayer_1, properties {soundWaveAssetPtr=&/sfx_supplydrop_reticle_appear}}]
properties {
 attenuationSettings=&/Fort_SupplyDrop_Reticle_Alert_Athena_Att
 maxDistance=75000
 soundClassObject=&Fort_Ambient_Athena
 volumeMultiplier=1.039
}
